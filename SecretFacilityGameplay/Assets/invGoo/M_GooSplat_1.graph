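# Translucent deferred decal material for a glossy green goo splat with a glowing, animated interior.
# A mask texture drives opacity, thickness and base color shading, over a flattened splat normal map.
# Glow combines parallax bubble normals lit by Fresnel with two panning noise samples, the first warping the second's UVs.
# Fresnel scales that tinted glow, which passes through inverse eye adaptation to counter auto-exposure.
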Material /SecretFacilityGameplay/Assets/invGoo/M_GooSplat_1
graph {
 properties {bAutomaticallySetUsageInEditor=false, blendMode=BLEND_Translucent, materialDomain=MD_DeferredDecal, usedShadingModels=Default Lit}
 nodes [
  Constant() {r=0.66}
  n1 = TextureSampleParameter2D() {parameterName=Normal, samplerType=SAMPLERTYPE_Normal, texture=&/GooSplat1_N}
  n2 = Fresnel(
   ExponentIn=ScalarParameter() {defaultValue=2.2, parameterName=Fresnel2 Exponent}
   BaseReflectFractionIn=Constant() {r=0.01}
   Normal=Transform(None=@n1.RGB)
  ) {baseReflectFraction=0.05, exponent=2}
  n3 = LinearInterpolate(A=ScalarParameter() {defaultValue=0.3528, parameterName=Fresnel Glow Blend Minimum}, Alpha=@n2) {constA=1}
  n4 = ScalarParameter() {defaultValue=0.25, parameterName=BubbleIntensity}
  n5 = ScalarParameter() {defaultValue=6.3007, parameterName=Fresnel3}
  n6 = Multiply(A=TextureCoordinate() {uTiling=2, vTiling=2}, B=ScalarParameter() {defaultValue=1.9161, parameterName=BubbleTiling})
  n7 = ScalarParameter() {defaultValue=-7.4899, parameterName=ubbleheight}
  n8 = Transform(None=TextureSample(UVs=BumpOffset(Coordinate=@n6, Height=@n7)) {samplerType=SAMPLERTYPE_Normal, texture=/Game/Items/Art_noLOD/Textures/Bubbles}.RGB)
  n9 = TextureSampleParameter2D() {parameterName=Mask, texture=&/GooSplat1_M}
  n10 = Multiply(A=Saturate(None=Fresnel(ExponentIn=@n5, Normal=@n8) {baseReflectFraction=0}), B=@n9.G)
  n11 = Multiply(A=TextureCoordinate(), B=ScalarParameter() {defaultValue=1.3365, parameterName=vavava}) {constB=3}
  n12 = TextureSampleParameter2D(UVs=BumpOffset(
    Coordinate=Panner(Coordinate=@n11) {speedX=0.01, speedY=0.02}
    Height=ScalarParameter() {defaultValue=-3.7904, parameterName=EmissiveLayer height}
   )) {parameterName=Mask1, texture=&/GooNoise_M}
  n13 = ScalarParameter() {defaultValue=0.3728, parameterName=Minimum Noise Brightness}
  n14 = TextureCoordinate()
  n15 = Add(A=Panner(Coordinate=Multiply(A=@n14)) {speedX=-0.01}, B=Multiply(A=@n12.B, B=ScalarParameter() {defaultValue=0.196, parameterName=Warp}))
  n16 = Abs(None=Multiply(A=TextureSampleParameter2D(UVs=@n15) {parameterName=Mask2, texture=&GooNoise_M}.G, B=@n2))
  n17 = Add(A=Multiply(A=Multiply(A=@n4, B=@n10), B=@n12.B), B=Saturate(None=LinearInterpolate(A=@n13, Alpha=@n16) {constA=0.2}))
  n18 = VectorParameter() {parameterName=Glow, defaultValue=<rgba 0.3424 0.75 0 1>}
  n19 = QualitySwitch(Default=@n18.RGB, Epic=Multiply(A=@n18.RGB, B=ScalarParameter() {defaultValue=1, parameterName=MobileSlimeGlowBoost}))
  EyeAdaptationInverse(
   LightValueInput=Multiply(A=Saturate(None=@n3), B=Multiply(A=@n17, B=@n19) {constA=1}) {constA=1}
   AlphaInput=ScalarParameter() {defaultValue=1, parameterName=UseInverseEyeAdaptation}
  )
  EyeAdaptation()
  MaterialFunctionCall(Normal=@n1.RGB, Flatness=ScalarParameter() {defaultValue=0.2915, parameterName=Normal Flatness}) {materialFunction=&FlattenNormal}
  Multiply()
  Multiply(
   A=LinearInterpolate(
    A=VectorParameter() {parameterName=dfasdf, defaultValue=<rgba 0.0017 0.0365 0.0265 1>}.RGB
    B=VectorParameter() {parameterName=sdf, defaultValue=<rgba 0.0376 0.0823 0 1>}.RGB
    Alpha=NamedRerouteUsage()
   )
   B=@n9.B
  )
  NamedRerouteDeclaration(None=@n2) {name=Fresnel, nodeColor=<rgba 0.0781 1 0 1>}
  n20 = Multiply(A=@n9.R, B=ScalarParameter() {defaultValue=0.88, parameterName=Opacity})
  n21 = LinearInterpolate(A=ScalarParameter() {defaultValue=0.7347, parameterName=Thickness}, Alpha=@n9.G)
  Reroute(None=Saturate(None=Multiply(A=@n20, B=@n21)))
  ScalarParameter() {defaultValue=0.073, parameterName=roughness}
  ScalarParameter() {defaultValue=0.8, parameterName=spec}
  n22 = Constant3Vector() {constant=<rgba 0 50 0 0>}
  ShadingPathSwitch(Default=@n19, Forward=@n22, Mobile=@n22)
 ]
}
