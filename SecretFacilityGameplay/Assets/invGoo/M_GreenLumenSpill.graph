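# Unlit material for a glowing green spill, apparently a luminous goo effect in the Secret Facility content.
# Two Perlin noise samples panning at different tiling and speeds multiply into a shifting brightness mask.
# That mask modulates the Glow color parameter, which a time-driven sine then pulses.
# The pulse phase is offset by each object's world X and Y position, so separate spills can throb out of sync.
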Material /SecretFacilityGameplay/Assets/invGoo/M_GreenLumenSpill
graph {
 properties {bAutomaticallySetUsageInEditor=false, shadingModel=MSM_Unlit, usedShadingModels=Unlit}
 nodes [
  n1 = TextureCoordinate()
  n2 = TextureSampleParameter2D(UVs=Panner(Coordinate=Multiply(A=@n1) {constB=1.5}) {speedX=0.16, speedY=0.08}) {parameterName=Mask2, texture=/Game/Effects/Fort_Effects/Textures/Tile/T_PerlinRGB_01_M}
  n3 = TextureCoordinate()
  n4 = TextureSampleParameter2D(UVs=Panner(Coordinate=Multiply(A=@n3) {constB=2.8}) {speedX=-0.2}) {parameterName=Mask2, texture=&T_PerlinRGB_01_M}
  n5 = Multiply(A=Abs(None=Multiply(A=@n2.B, B=@n4.B))) {constB=2}
  n6 = Time()
  n7 = ObjectPositionWS()
  n8 = Add(A=ComponentMask(None=@n7) {r=true}, B=ComponentMask(None=@n7) {g=true})
  n9 = Divide(A=Abs(None=Multiply(A=@n8) {constB=484})) {constB=40000}
  n10 = ConstantBiasScale(None=Sine(None=Add(A=@n6, B=@n9)) {period=5}) {bias=0.5, scale=1}
  LinearInterpolate(
   B=Multiply(A=LinearInterpolate(Alpha=@n5) {constA=0.6}, B=VectorParameter() {parameterName=Glow, defaultValue=<rgba 2.8659 5 0 1>}.RGB)
   Alpha=Multiply(A=LinearInterpolate(Alpha=@n10) {constA=0.5})
  )
 ]
}
